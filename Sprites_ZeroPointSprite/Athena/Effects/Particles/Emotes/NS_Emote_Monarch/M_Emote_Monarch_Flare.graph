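# Unlit additive material for Niagara sprites that draws a camera flare in the Monarch emote effect.
# Particle color RGB passes through the MF_FX_EmissiveAdjustments function to shape the glow.
# A grayscale flare texture, softened by a depth fade with adjustable distance, is multiplied by particle alpha to form the fade mask.
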Material /Sprites_ZeroPointSprite/Athena/Effects/Particles/Emotes/NS_Emote_Monarch/M_Emote_Monarch_Flare
graph {
 properties {bUsedWithNiagaraSprites=true, blendMode=BLEND_Additive, pixelDepthOffsetMode=PDOM_Legacy, shadingModel=MSM_Unlit, usedShadingModels=Unlit}
 nodes [
  n1 = ParticleColor()
  MaterialFunctionCall(Emissive=@n1.RGB) {materialFunction=&MF_FX_EmissiveAdjustments}
  Multiply(
   A=@n1.A
   B=DepthFade(
    Opacity=TextureSampleParameter2D() {parameterName=MainTex, samplerType=SAMPLERTYPE_Grayscale, texture=/Game/Effects/Fort_Effects/Textures/T_EFX_CameraFlare_01}.R
    FadeDistance=ScalarParameter() {defaultValue=33, parameterName=Depth Fade Distance}
   )
  )
 ]
}
